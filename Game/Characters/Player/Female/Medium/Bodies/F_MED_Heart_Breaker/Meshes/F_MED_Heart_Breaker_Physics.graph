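# Physics asset for the Heart Breaker medium female player body, covering its dangling belt, sword and chain bones.
# One linear constraint series runs from belt base through belt, sword and chain parent to four chain links, using boxes with a sphere at the tip.
# Tight, partly locked joints near the belt give way to wider swing limits toward the tip.
# An unconstrained right-thigh capsule provides a leg collision shape.
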PhysicsAsset /Game/Characters/Player/Female/Medium/Bodies/F_MED_Heart_Breaker/Meshes/F_MED_Heart_Breaker_Physics
bodies {
 dyn_belt_base [{shapeType=Box, extentX=0.5, extentY=0.5, extentZ=0.5}]
 dyn_belt_parent [{shapeType=Box, extentX=11.3369, extentY=14.0234, extentZ=2.5346, center=<5.8378 -0.0001 0>}]
 dyn_chain_01 [{shapeType=Box, extentX=2, extentY=2, extentZ=1, center=<1.5 0 0>}]
 dyn_chain_02 [{shapeType=Box, extentX=2, extentY=2, extentZ=1, center=<1.5 0 0>}]
 dyn_chain_03 [{shapeType=Box, extentX=2, extentY=2, extentZ=1, center=<1.5 0 0>}]
 dyn_chain_04 [{shapeType=Sphere, radius=3, center=<4 0 0>}]
 dyn_chain_parent [{shapeType=Box, extentX=2, extentY=2, extentZ=1, center=<1.5 0 0>}]
 dyn_sword_parent [{shapeType=Box, extentX=11, extentY=20, extentZ=3, center=<5 0 0>}]
 thigh_r [{shapeType=Capsule, length=22.125, radius=7.875, center=<16.8684 -0.0973 0.3683>, rotation=<rot -89.7451 179.9997 175.2874>}]
}
constraints [
 {
  bone1Name=dyn_belt_parent
  bone2Name=dyn_belt_base
  swing1LimitDegrees=5
  swing1Motion=Locked
  swing2LimitDegrees=2
  twistLimitDegrees=2
  twistMotion=Locked
 }
 {bone1Name=dyn_sword_parent, bone2Name=dyn_belt_parent, swing1LimitDegrees=6, swing2LimitDegrees=10, twistLimitDegrees=1}
 {bone1Name=dyn_chain_parent, bone2Name=dyn_sword_parent, swing1LimitDegrees=2, swing2LimitDegrees=10, twistLimitDegrees=1, twistMotion=Locked}
 {bone1Name=dyn_chain_01, bone2Name=dyn_chain_parent, swing1LimitDegrees=10, swing2LimitDegrees=10, twistLimitDegrees=3}
 {bone1Name=dyn_chain_02, bone2Name=dyn_chain_01, swing1LimitDegrees=10, swing2LimitDegrees=10, twistLimitDegrees=3}
 {bone1Name=dyn_chain_03, bone2Name=dyn_chain_02, swing1LimitDegrees=15, swing2LimitDegrees=15, twistLimitDegrees=3}
 {bone1Name=dyn_chain_04, bone2Name=dyn_chain_03, swing1LimitDegrees=25, swing2LimitDegrees=25, twistLimitDegrees=3}
]
