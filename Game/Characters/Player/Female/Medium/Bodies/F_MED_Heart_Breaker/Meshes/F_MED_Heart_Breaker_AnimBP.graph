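# Post-process-style Anim Blueprint for the Heart Breaker female body, adding secondary motion to a pose copied from the attached parent mesh.
# A deform Control Rig feeds sleeve AnimDynamics, a weighted belt constraint with crouch offset, then RigidBody simulation.
# Thigh helper bones scale with leg swing angles computed each update.
# Weapon, customization and skydiving events toggle the "Hide Element 01" material parameter using the matching pickaxe's meta tag.
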AnimBlueprint /Game/Characters/Player/Female/Medium/Bodies/F_MED_Heart_Breaker/Meshes/F_MED_Heart_Breaker_AnimBP
target_skeleton=&/F_MED_Heart_Breaker_Skeleton
graphs [
 AnimGraph {nodes [
   "Transform (Modify) Bone - Bone: dyn_belt_parent"(Alpha=Get IsJumpingInAir() -> (IsJumpingInAir:double), Translation="0,0,0", Rotation="0,10,0", Scale="1,1,1") {
    rotationMode=BMM_Additive
    rotationSpace=BCS_BoneSpace
    alphaBoolBlend {blendOption=Linear}
    boneToModify {boneName=dyn_belt_parent}
    rotation {yaw=10}
   }
   n1 = Control Rig(Source=Copy Pose From Mesh() {bUseAttachedParent=true}) {
    alphaBoolBlend {blendOption=Linear}
    controlRigAssetReference {blueprintRigClass=&/F_MED_Heart_Breaker_DeformRig_ControlRig.F_MED_Heart_Breaker_DeformRig_ControlRig_C}
   }
   n2 = Anim Dynamics - Bone: dyn_sleeve_r_inner(ComponentPose=Local To Component(LocalPose=@n1), Alpha=1, ExternalForce="0,0,0") {
    angularDampingOverride=0.999
    angularSpringConstant=5000
    bAngularSpring=true
    bGravityOverrideInSimSpace=true
    bOverrideAngularDamping=true
    gravityScale=0.35
    simulationSpace=RootRelative
    alphaBoolBlend {blendOption=Linear}
    alphaScaleBiasClamp {scale=0.2}
    boundBone {boneName=dyn_sleeve_r_inner}
    componentAppliedLinearAccClamp=<6000 6000 6000>
    componentLinearAccScale=<0.25 0.25 0.6>
    componentLinearVelScale=<0.25 0.25 0.6>
    physicsBodyDefinitions [{
      sphereCollisionRadius=0
      boundBone {boneName=dyn_sleeve_r_inner}
      boxExtents=<8 25 25>
      constraintSetup {angularLimitsMax=<45 20 20>, angularLimitsMin=<-45 -20 -75>, angularTarget=<1 0 0>}
      localJointOffset=<5 0 0>
     }]
   }
   n3 = Anim Dynamics - Bone: dyn_sleeve_r_lower(ComponentPose=@n2, Alpha=1, ExternalForce="0,0,0") {
    angularDampingOverride=0.999
    angularSpringConstant=5000
    bAngularSpring=true
    bGravityOverrideInSimSpace=true
    bOverrideAngularDamping=true
    gravityScale=0.35
    simulationSpace=RootRelative
    alphaBoolBlend {blendOption=Linear}
    alphaScaleBiasClamp {scale=0.2}
    boundBone {boneName=dyn_sleeve_r_lower}
    componentAppliedLinearAccClamp=<6000 6000 6000>
    componentLinearAccScale=<0.25 0.25 0.6>
    componentLinearVelScale=<0.25 0.25 0.6>
    physicsBodyDefinitions [{
      sphereCollisionRadius=0
      boundBone {boneName=dyn_sleeve_r_lower}
      boxExtents=<8 25 25>
      constraintSetup {angularLimitsMax=<45 20 20>, angularLimitsMin=<-45 -20 -75>, angularTarget=<1 0 0>}
      localJointOffset=<5 0 0>
     }]
   }
   n4 = Anim Dynamics - Bone: dyn_sleeve_r_outter(ComponentPose=@n3, Alpha=1, ExternalForce="0,0,0") {
    angularDampingOverride=0.999
    angularSpringConstant=5000
    bAngularSpring=true
    bGravityOverrideInSimSpace=true
    bOverrideAngularDamping=true
    gravityScale=0.35
    simulationSpace=RootRelative
    alphaBoolBlend {blendOption=Linear}
    alphaScaleBiasClamp {scale=0.2}
    boundBone {boneName=dyn_sleeve_r_outter}
    componentAppliedLinearAccClamp=<6000 6000 6000>
    componentLinearAccScale=<0.25 0.25 0.6>
    componentLinearVelScale=<0.25 0.25 0.6>
    physicsBodyDefinitions [{
      sphereCollisionRadius=0
      boundBone {boneName=dyn_sleeve_r_outter}
      boxExtents=<8 25 25>
      constraintSetup {angularLimitsMax=<45 40 20>, angularLimitsMin=<-45 -20 -75>, angularTarget=<1 0 0>}
      localJointOffset=<5 0 0>
     }]
   }
   n5 = Anim Dynamics - Bone: dyn_sleeve_l_inner(ComponentPose=@n4, Alpha=1, ExternalForce="0,0,0") {
    angularDampingOverride=0.999
    angularSpringConstant=5000
    bAngularSpring=true
    bGravityOverrideInSimSpace=true
    bOverrideAngularDamping=true
    gravityScale=0.35
    simulationSpace=RootRelative
    alphaBoolBlend {blendOption=Linear}
    alphaScaleBiasClamp {scale=0.2}
    boundBone {boneName=dyn_sleeve_l_inner}
    componentAppliedLinearAccClamp=<6000 6000 6000>
    componentLinearAccScale=<0.25 0.25 0.6>
    componentLinearVelScale=<0.25 0.25 0.6>
    physicsBodyDefinitions [{
      sphereCollisionRadius=0
      boundBone {boneName=dyn_sleeve_l_inner}
      boxExtents=<8 25 25>
      constraintSetup {angularLimitsMax=<45 20 20>, angularLimitsMin=<-45 -20 -75>, angularTarget=<1 0 0>}
      localJointOffset=<5 0 0>
     }]
   }
   n6 = Anim Dynamics - Bone: dyn_sleeve_l_lower(ComponentPose=@n5, Alpha=1, ExternalForce="0,0,0") {
    angularDampingOverride=0.999
    angularSpringConstant=5000
    bAngularSpring=true
    bGravityOverrideInSimSpace=true
    bOverrideAngularDamping=true
    gravityScale=0.35
    simulationSpace=RootRelative
    alphaBoolBlend {blendOption=Linear}
    alphaScaleBiasClamp {scale=0.2}
    boundBone {boneName=dyn_sleeve_l_lower}
    componentAppliedLinearAccClamp=<6000 6000 6000>
    componentLinearAccScale=<0.25 0.25 0.6>
    componentLinearVelScale=<0.25 0.25 0.6>
    physicsBodyDefinitions [{
      sphereCollisionRadius=0
      boundBone {boneName=dyn_sleeve_l_lower}
      boxExtents=<8 25 25>
      constraintSetup {angularLimitsMax=<45 20 20>, angularLimitsMin=<-45 -20 -75>, angularTarget=<1 0 0>}
      localJointOffset=<5 0 0>
     }]
   }
   n7 = Anim Dynamics - Bone: dyn_sleeve_l_outter(ComponentPose=@n6, Alpha=1, ExternalForce="0,0,0") {
    angularDampingOverride=0.999
    angularSpringConstant=5000
    bAngularSpring=true
    bGravityOverrideInSimSpace=true
    bOverrideAngularDamping=true
    gravityScale=0.35
    simulationSpace=RootRelative
    alphaBoolBlend {blendOption=Linear}
    alphaScaleBiasClamp {scale=0.2}
    boundBone {boneName=dyn_sleeve_l_outter}
    componentAppliedLinearAccClamp=<6000 6000 6000>
    componentLinearAccScale=<0.25 0.25 0.6>
    componentLinearVelScale=<0.25 0.25 0.6>
    physicsBodyDefinitions [{
      sphereCollisionRadius=0
      boundBone {boneName=dyn_sleeve_l_outter}
      boxExtents=<8 25 25>
      constraintSetup {angularLimitsMax=<45 20 20>, angularLimitsMin=<-45 -20 -75>, angularTarget=<1 0 0>}
      localJointOffset=<5 0 0>
     }]
   }
   n8 = "Constraint - dyn_belt_base (PPP)"!(ComponentPose=@n7, Alpha=1, ConstraintWeights_0=0.031, ConstraintWeights_1:float=0.5454, ConstraintWeights_2:float=0.4237) -> (Pose) {
    alphaBoolBlend {blendOption=Linear}
    boneToModify {boneName=dyn_belt_base}
    constraintSetup [{targetBone {boneName=thigh_r}}, {targetBone {boneName=pelvis}}, {targetBone {boneName=thigh_twist_01_r}}]
    constraintWeights [1,1,1]
   }
   n9 = "Transform (Modify) Bone - Bone: dyn_belt_base"(ComponentPose=@n8, Alpha=Get Crouch_Alpha() -> (Crouch_Alpha:double), Translation="0,0,-2", Rotation="-2,0,0", Scale="1,1,1") {
    rotationMode=BMM_Additive
    rotationSpace=BCS_BoneSpace
    translationMode=BMM_Additive
    translationSpace=BCS_BoneSpace
    alphaBoolBlend {blendOption=Linear}
    boneToModify {boneName=dyn_belt_base}
    rotation {pitch=-2}
    translation {z=-2}
   }
   n10 = RigidBody!(ComponentPose=@n9, Alpha, OverrideWorldGravity:Vector=Get Override World Gravity() -> (Override World Gravity:Vector), ExternalForce) -> (Pose) {
    bClampLinearTranslationLimitToRefPose=true
    bOverrideWorldGravity=true
    overridePhysicsAsset=&/F_MED_Heart_Breaker_Physics
    simulationSpace=BaseBoneSpace
    alphaBoolBlend {blendOption=Linear}
    alphaScaleBiasClamp {bInterpResult=true, interpSpeedDecreasing=3, interpSpeedIncreasing=2}
    baseBoneRef {boneName=spine_01}
    componentLinearAccScale=<0.3 0.3 0.3>
    componentLinearVelScale=<0.43 0.43 0.43>
    overrideWorldGravity {z=-980}
    simSpaceSettings {worldAlpha=0.05}
   }
   n11 = "Transform (Modify) Bone - Bone: dyn_sub_thigh_l"(
    ComponentPose=@n10
    Alpha=Get legSwingDelta_l() -> (legSwingDelta_l:double)
    Translation="0,0,0"
    Rotation="0,0,0"
    Scale="5.750,3.750,12.875"
   ) {
    scaleMode=BMM_Additive
    scaleSpace=BCS_BoneSpace
    alphaBoolBlend {blendOption=Linear}
    boneToModify {boneName=dyn_sub_thigh_l}
    scale=<5.75 3.75 12.875>
   }
   n12 = "Transform (Modify) Bone - Bone: dyn_sub_thigh_r"(ComponentPose=@n11, Alpha=Get legSwingDelta_r() -> (legSwingDelta_r:double), Translation="0,0,0", Rotation="0,0,0", Scale="8,6,12.875") {
    scaleMode=BMM_Additive
    scaleSpace=BCS_BoneSpace
    alphaBoolBlend {blendOption=Linear}
    boneToModify {boneName=dyn_sub_thigh_r}
    scale=<8 6 12.875>
   }
   Output Pose(Result=Component To Local(ComponentPose=@n12)) {layerGroup=None}
  ]}
 EventGraph {nodes [
   Event BlueprintUpdateAnimation()
   > n1 = Sequence!(execute) -> (then_0, then_1, then_2:Exec, then_3:Exec)
   > Set Crouch_Alpha(Crouch_Alpha:double=SelectFloat(A=1, B=0, bPickA=Get bIsCrouching())) -> (Output_Get:double)
   n2 = Branch(execute=@n1.then_2, Condition=Get bIsJumping())
   > Set IsJumpingInAir(IsJumpingInAir:double=1) -> (Output_Get:double)
   Set IsJumpingInAir(execute=@n2.else, IsJumpingInAir=0)
   n3 = GetOwnerSkeletalMeshComponent()
   Set legSwingDelta_r(
    execute=@n1.then_1
    legSwingDelta_r:double=MapRangeClamped(Value=GetTwistAndSwingAngleOfDeltaRotationFromRefPose(self=@n3, BoneName=thigh_r).OutSwingAngle, InRangeA=150, InRangeB=75)
   ) -> (Output_Get:double)
   > Set legSwingDelta_l(legSwingDelta_l:double=MapRangeClamped(Value=GetTwistAndSwingAngleOfDeltaRotationFromRefPose(self=@n3, BoneName=thigh_l).OutSwingAngle, InRangeA=43, InRangeB=110)) -> (Output_Get:double)
   n4 = Event OnWeaponChanged()
   n5 = Get Dynamic Material Instance(execute=[@n4.then, @n13.then], SkeletalMeshComponent=GetPartSkeletalMeshComponent())
   > n6 = Sequence!(execute) -> (then_0, then_1)
   n7 = PickaxeItemDefinitionHasAnyMetaTag(Tags=Make Array("[0]"=(TagName="Cosmetics.MetaTags.HeartBreakerFemale")), bRequiresEquipped=true)
   > n8 = Branch(Condition=@n7)
   > n9 = Branch(Condition=Get bIsSkyDiving())
   SetScalarParameterValue(execute=[@n11.then, @n9.then, @n10.else, @n8.else], self=@n5, ParameterName=Hide Element 01, Value=0)
   SetScalarParameterValue(execute=[@n9.else, @n10.then], self=@n5, ParameterName=Hide Element 01, Value=1)
   n10 = Branch(execute=[@n12.then, @n6.then_1], Condition=@n7)
   n11 = Event OnBeginSkydiving()
   n12 = Event OnEndSkydiving()
   n13 = Event OnCharacterCustomizationLoaded()
  ]}
]
variables [
 Crouch_Alpha {type=number}
 Game Time in Seconds {type=number}
 IsJumpingInAir {type=number}
 legSwingDelta_l {type=number}
 legSwingDelta_r {type=number}
 Override World Gravity {type=Vector, default=<0 0 -980>}
]
